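; Sequence runner for Day 2 of the GrillMoon countdown, broadcasting its day state through stateful Event Router events.
; Calendar initialization sends the day event; received events either start POI music and ambience or fast-forward the day to near finished, lighting the mask.
; The GM_Day2 cheat sends the day event and jumps to a given percentage.
; Both skip paths pre-load on the server, which plays the sequence once loading completes.
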(event Custom|GM_Day2
  (params (Percentage :type Float%20%28double-precision%29))
  ;[0 0 0] Console Command / Load and Play
  (Variables|Cheats|SetCheat%20Day%20Percentage Percentage) ; Allows skipping ahead in the\nsequence for testing. A value\nbetween 0 - 1.0
  (CallFunction|SendDayEvent "NewEnumerator1")
  (if (Networking|IsServer)
    (Variables|Cheats|SetStarting%20Method "NewEnumerator1") ; Keeps some calendar logic from running
    (Variables|Default|SetRun%20From%20Cheat true)
    (CallFunction|%28CallOnServer%29Pre-LoadSequence)))

(event Custom|Set%20This%20Day%20to%20Finished
  ;[0 0 0] Console Command / Load and Play
  (Variables|Cheats|SetCheat%20Day%20Percentage 0.95) ; set to near finished so the mask lights up
  (Utilities|FlowControl|DoOnce false
    (:Completed
      (if (Networking|IsServer)
        (Variables|Cheats|SetStarting%20Method "NewEnumerator1") ; Keeps some calendar logic from running
        (Variables|Default|SetRun%20From%20Cheat true)
        (CallFunction|%28CallOnServer%29Pre-LoadSequence)))))

(event EventPre-LoadSequenceComplete
  ;[0 0 0] Console Command / Load and Play
  (|CallParentFunction)
  (if (Networking|IsServer)
    (if (Variables|Default|GetRun%20From%20Cheat)
      (CallFunction|LoadSequence) ; Play the sequence.
    )))

(event EventInitComplete-PreCalendarCheck
  ;[0 0 0] Console Command / Load and Play
  (|CallParentFunction)
  (CallFunction|BindDayEvent))

(event EventInitializefromCalendarEvent
  ;[0 0 0] Console Command / Load and Play
  (|CallParentFunction)
  (CallFunction|SendDayEvent "NewEnumerator1"))

(event Custom|Bind%20Day%20Event
  ;[0.3872 0.4056 1 0.5] Catch Event Router Events
  (bind (_asynctaskproxy _payload) (GameplayEventRouter|StartListeningToStatefulEvent (GameplayEventRouter|GetGlobalEventRouter) "/GrillMoon_Countdown/Gameplay/EventRouter/GrillMoon_Day_Event.GrillMoon_Day_Event")
    (:OnEventReceived
      _)
    (:OnSavedEventStateExists
      _))
  (switch Utilities|FlowControl|Switch|SwitchonGrillMoon_Day_Event_Enum (Utilities|Struct|BreakGrillMoonDayEvent _payload)
    (:NewEnumerator17
      (Class|BPGrillMoonPOIMusicComponent|StartPOIMusic (Variables|Default|GetBP_GrillMoon_POI_MusicComponent))
      (Class|BPGrillMoonPOIAmbienceComponent|SpawnPOIAmbience (Variables|Default|GetBP_GrillMoon_POI_AmbienceComponent)))
    (:NewEnumerator18
      (CallFunction|SetThisDayToFinished) ; activate the mask quickly
    )))

(event Custom|Send%20Day%20Event
  (params (Day :type GrillMoon_Day_Event_Enum%20Enum :default "NewEnumerator17"))
  ;[0.3872 0.4056 1 0.5] Catch Event Router Events
  (GameplayEventRouter|BroadcastStatefulEvent (GameplayEventRouter|GetGlobalEventRouter) (Utilities|Struct|MakeGrillMoonDayEvent Day) 0))

(fn ConstructionScript
  (|CallParentFunction))
