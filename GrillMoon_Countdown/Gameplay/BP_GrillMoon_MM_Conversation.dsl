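; Interactable MistMylar conversation actor for the GrillMoon countdown event, offering "Talk to" unless the player is already in a conversation.
; On authority, after match-ready and calendar sync, it watches calendar events and a stateful day-event router, mapping both to conversation entry tags.
; The event zone's end or a specific day-event value sets CalendarEventOver, whose OnRep calls TurnOffActor to disable conversations, indicator bubble, interaction and mesh collision.
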(event EventBeginPlay
  (bind _daybasedconvotagscalendar (Variables|Default|GetDayBasedConvoTagsCalendar))
  (bind _daybasedconvotags (Variables|Default|GetDayBasedConvoTags))
  (Networking|SwitchHasAuthority
    (:Authority
      (Class|FortAsyncActionAuthorityMatchReady|MatchReadyAsync
        (:OnReady
          (MCPContext|InitialCalendarSyncCompleted
            (:OnSyncComplete
              (bind _eventname (MCPContext|WatchCalendarEvents (Variables|Default|GetCalendarEventToWatch))
                (:then
                  (bind _biseventactive (CalendarEvents|GetCalendarEventInformation "GrillMoon_EventZoneActive"))
                  (if _biseventactive
                    (Variables|/Script/FortniteGame.FortNonPlayerConversationParticipantComponent|SetConversationEntryTag "(TagName=\"Conversation.Entry.MistMylar.GrillMoon.Day11.Event\")" (Variables|Default|GetNonPlayerConversationComponent))
                    (else
                      _))
                  (bind (_asynctaskproxy _payload) (GameplayEventRouter|StartListeningToStatefulEvent (GameplayEventRouter|GetGlobalEventRouter) "/GrillMoon_Countdown/Gameplay/EventRouter/GrillMoon_Day_Event.GrillMoon_Day_Event")
                    (:OnEventReceived
                      (if (.ReturnValue (Utilities|Map|Find _daybasedconvotags (Utilities|Struct|BreakGrillMoonDayEvent _payload)))
                        (Variables|/Script/FortniteGame.FortNonPlayerConversationParticipantComponent|SetConversationEntryTag (Utilities|Map|Find _daybasedconvotags (Utilities|Struct|BreakGrillMoonDayEvent _payload)) (Variables|Default|GetNonPlayerConversationComponent))
                        (else
                          _))
                      (if (Utilities|Enum|Equal(Enum) (Utilities|Struct|BreakGrillMoonDayEvent _payload) "NewEnumerator13")
                        (Variables|Default|SetCalendarEventOver true)))))
                (:OnEventActive
                  (if (.ReturnValue (Utilities|Map|Find _daybasedconvotagscalendar _eventname))
                    (Variables|/Script/FortniteGame.FortNonPlayerConversationParticipantComponent|SetConversationEntryTag (Utilities|Map|Find _daybasedconvotagscalendar _eventname) (Variables|Default|GetNonPlayerConversationComponent))))
                (:OnEventEnded
                  (if (Utilities|String|EqualExactly(String) _eventname "GrillMoon_EventZoneActive")
                    (Variables|Default|SetCalendarEventOver true)))))))))))

(event Custom|TurnOffActor
  (bind _nonplayerconversationcomponent (Variables|Default|GetNonPlayerConversationComponent))
  (Variables|/Script/FortniteGame.FortNonPlayerConversationParticipantComponent|SetbConversationsEnabled false _nonplayerconversationcomponent)
  (Variables|/Script/FortniteGame.FortNonPlayerConversationParticipantComponent|SetbShowIndicatorBubble false _nonplayerconversationcomponent)
  (Variables|Interaction|SetbAllowInteract false)
  (Collision|SetCollisionEnabled (Variables|Debug|GetStaticMeshComponent) "NoCollision"))

(event Custom|GM_EventZoneConversationChange
  (Variables|/Script/FortniteGame.FortNonPlayerConversationParticipantComponent|SetConversationEntryTag "(TagName=\"Conversation.Entry.MistMylar.GrillMoon\")" (Variables|Default|GetNonPlayerConversationComponent)))

(fn BlueprintCanInteract
  (params (InteractingPawn :type Fort%20Pawn%20Object%20Reference) (InteractionBeingAttempted :type EInteractionBeingAttempted%20Enum :default "FirstInteraction") (InteractionType :type TInteractionType%20Enum :default "IT_NoInteraction"))
  (if (and (Variables|Interaction|GetbAllowInteract) (not (Variables|BuildingActor|GetbDestroyed)))
    (bind _returnvalue (Conversation|FindConversationComponent (Pawn|GetController InteractingPawn)))
    (Utilities|IsValid _returnvalue
      (:Is%20Valid
        (return (not (Conversation|IsinActiveConversation _returnvalue))))
      (:Is%20Not%20Valid
        (return true)))
    (else
      (return false))))

(fn BlueprintGetInteractionString
  (params (InteractingPawn :type Fort%20Pawn%20Object%20Reference) (InteractionBeingAttempted :type EInteractionBeingAttempted%20Enum :default "FirstInteraction"))
  (return "NSLOCTEXT(\"[88B2DFA31BA2DD292CA85861C0FCE0FF]\", \"6F124B7C47BEA9329C5B8BB10914F4E7\", \"Talk to\")"))

(fn BlueprintGetInteractionTime
  (params (InteractingPawn :type Fort%20Pawn%20Object%20Reference) (InteractionBeingAttempted :type EInteractionBeingAttempted%20Enum :default "FirstInteraction"))
  (return (Variables|Default|GetInteractSpeedOverride) true))

(fn GetDayEventConversationTag
  (params (ActiveCalendarEvent :type String))
  (if (Utilities|String|EqualExactly(String) ActiveCalendarEvent (Variables|Default|GetDay11Event))
    (for _array_element (Variables|Default|GetCalendarEventToWatch)
      (if (Utilities|String|NotEqualExactly(String) (Variables|Default|GetDay11Event) _array_element)
        _))
    (else
      (return (Utilities|Map|Find (Variables|Default|GetDayBasedConvoTagsCalendar) ActiveCalendarEvent)))))

(fn GetMinDistanceFromInteraction
  (return (Variables|Default|GetInteractionDistance)))

(fn OnRep_CalendarEventOver
  (if (Variables|Default|GetCalendarEventOver)
    (CallFunction|TurnOffActor)))

(fn ConstructionScript)
