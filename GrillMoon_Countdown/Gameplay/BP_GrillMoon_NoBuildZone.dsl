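; BP_GrillMoon_NoBuildZone is a box volume that blocks player building inside it during the GrillMoon Countdown event.
; Once playlist data is ready, the box's Pawn collision response switches to Overlap.
; On the server, a player pawn entering the box has its building tool force-cancelled and receives the GE_GrillMoon_BlockBuildMode gameplay effect.
; Leaving the box removes that effect on the server.
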(event EventBeginPlay
  ;[1 1 1] Active only during the event
  (bind (_gamestate _playlist _playlistcontexttags) (Class|FortAsyncActionCurrentPlaylistReady|PlaylistDataReadyAsync)
    (:OnReady
      (Collision|SetCollisionResponsetoChannel (Variables|Default|GetBox) "ECC_Pawn" "ECR_Overlap"))))

(event OnComponentBeginOverlap(Box)
  (params (OverlappedComponent :type Primitive%20Component%20Object%20Reference) (OtherActor :type Actor%20Object%20Reference) (OtherComp :type Primitive%20Component%20Object%20Reference) (OtherBodyIndex :type Integer) (bFromSweep :type Boolean) (SweepResult :type Hit%20Result%20Structure%20%28by%20ref%29))
  ;[1 1 1] Force cancel building and block building after overlapping with collision
  (if (Networking|IsServer)
    (bind _asfort_player_pawn (Utilities|Casting|CastToFortPlayerPawn OtherActor))
    (Building|ForceCancelBuildingTool (Fortnite|GetFortPlayerControllerfromActor _asfort_player_pawn))
    (GameplayEffects|ApplyGameplayEffectToTarget :self (Ability|GetAbilitySystemComponent _asfort_player_pawn) :GameplayEffectClass "/GrillMoon_Countdown/GameplayCues/GE_GrillMoon_BlockBuildMode.GE_GrillMoon_BlockBuildMode_C" :Target (Ability|GetAbilitySystemComponent _asfort_player_pawn) :Level 0.0)))

(event OnComponentEndOverlap(Box)
  (params (OverlappedComponent :type Primitive%20Component%20Object%20Reference) (OtherActor :type Actor%20Object%20Reference) (OtherComp :type Primitive%20Component%20Object%20Reference) (OtherBodyIndex :type Integer))
  ;[1 1 1] Remove build blocking when ending overlap with collision
  (if (Networking|IsServer)
    (bind _asfort_player_pawn (Utilities|Casting|CastToFortPlayerPawn OtherActor))
    (GameplayEffects|RemoveActiveGameplayEffectbySourceEffect (Ability|GetAbilitySystemComponent _asfort_player_pawn) "/GrillMoon_Countdown/GameplayCues/GE_GrillMoon_BlockBuildMode.GE_GrillMoon_BlockBuildMode_C" (Ability|GetAbilitySystemComponent _asfort_player_pawn) -1)))

(fn ConstructionScript)
